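# Character is a SoundClass in the Figment_S07 sound hierarchy that groups character audio, nested under the shared Fort_Enemy_Attack class.
# It parents three child classes: a critical enemy attack class marked DELETE_ME, a non-critical enemy class and a troll class.
# Its own properties enable effects processing and send some bleed to the LFE channel.
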SoundClass /Figment_S07/Sounds/FortSoundClassesAndModes/Character
parent=/Game/Sounds/FortSoundClassesAndModes/Fort_Enemy_Attack
children [
 /Game/Sounds/FortSoundClassesAndModes/DELETE_ME_Fort_Enemy_Attack_Critical_PSM
 /Game/Sounds/FortSoundClassesAndModes/Fort_Enemy_NonCritical
 &/Fort_Troll
]
properties {properties {bApplyEffects=true, lFEBleed=0.25}}
