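# Sound class for open-space ambience in Figment S07, parented to the Foundations_Ambience class.
# A passive sound mix modifier activates Foundations_DuckAmbience_OpenSpace while sounds of this class play within a broad volume threshold window, so their playback drives that mix automatically.
# The class also enables effects and sets a small LFE bleed.
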SoundClass /Figment_S07/Sounds/FortSoundClassesAndModes/Foundations_Ambience_OpenSpace_PSM
parent=&/Foundations_Ambience
properties {
 passiveSoundMixModifiers [{maxVolumeThreshold=1000, minVolumeThreshold=0, soundMix=&/Foundations_DuckAmbience_OpenSpace}]
 properties {bApplyEffects=true, lFEBleed=0.2}
}
